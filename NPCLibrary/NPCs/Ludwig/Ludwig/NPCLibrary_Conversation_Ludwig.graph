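# Conversation database for the NPC Ludwig, defining the dialogue players get when talking to this brawl-hungry character.
# A single entry point leads to a Speech node with a default line and contextual variants chosen by requirement.
# Variants cover a first-ever meeting, prioritized through a curve row, and reactions to specific player cosmetic character IDs.
# The speech then branches through link nodes into the shared NPC library quests and basic services conversations.
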ConversationDatabase /NPCLibrary/NPCs/Ludwig/Ludwig/NPCLibrary_Conversation_Ludwig
entry_points [{tag=Conversation.Entry.Ludwig, destinations [#1]}]
nodes [
 ConversationEntryPointNode_7 {
  id=#1
  class=ConversationEntryPointNode
  outputs [#2]
  properties {entryTag {tagName=Conversation.Entry.Ludwig}}
 }
 Speech_C_4 {
  id=#2
  class=Speech_C
  outputs [#3, #4]
  properties {
   nodeName=""
   generalConfig {
    defaultMessage=I'm in the mood to brawl...
    contextualMessages [
     {
      message=This war has awoken something in me...
      contextRequirements [{requirement {class=FortControllerRequirement_IsFirstEverConversationWithNPC}}]
      priority {curve {rowName=Tandem.Speech.Priority.FirstEverInteraction}}
     }
     {
      message=Your destiny say anything about a guy who can't stop fighting?
      contextRequirements [{requirement {
         class=FortControllerRequirement_HasCID
         properties {allowedCIds [/BRCosmetics/Athena/Items/Cosmetics/Characters/CID_A_424_Athena_Commando_M_Lancelot]}
        }}]
      priority {value=0, curve {curveTable=null, rowName=None}}
     }
     {
      message=Looking for an apprentice assassin?
      contextRequirements [{requirement {
         class=FortControllerRequirement_HasCID
         properties {allowedCIds [/BRCosmetics/Athena/Items/Cosmetics/Characters/CID_A_428_Athena_Commando_F_PinkWidow]}
        }}]
      priority {value=0, curve {curveTable=null, rowName=None}}
     }
     {
      message="Sure, I'll joust you."
      contextRequirements [{requirement {
         class=FortControllerRequirement_HasCID
         properties {allowedCIds [/BRCosmetics/Athena/Items/Cosmetics/Characters/CID_A_427_Athena_Commando_F_Fuchsia]}
        }}]
      priority {value=0, curve {curveTable=null, rowName=None}}
     }
     {
      message=You look like a fighter.
      contextRequirements [{requirement {
         class=FortControllerRequirement_HasCID
         properties {allowedCIds [/BRCosmetics/Athena/Items/Cosmetics/Characters/CID_A_429_Athena_Commando_M_Collectable]}
        }}]
      priority {value=0, curve {curveTable=null, rowName=None}}
     }
     {
      message=Is the Peace Syndicate recruiting?
      contextRequirements [{requirement {
         class=FortControllerRequirement_HasCID
         properties {allowedCIds [&CID_A_425_Athena_Commando_F_BlueJay]}
        }}]
      priority {value=0, curve {curveTable=null, rowName=None}}
     }
     {
      message="You wanna fight, doctor?"
      contextRequirements [{requirement {
         class=FortControllerRequirement_HasCID
         properties {allowedCIds [/BRCosmetics/Athena/Items/Cosmetics/Characters/CID_A_423_Athena_Commando_M_Canary]}
        }}]
      priority {value=0, curve {curveTable=null, rowName=None}}
     }
     {
      message=Maybe I SHOULD embrace the dark side...
      contextRequirements [{requirement {
         class=FortControllerRequirement_HasCID
         properties {allowedCIds [&CID_A_422_Athena_Commando_M_Realm]}
        }}]
      priority {value=0, curve {curveTable=null, rowName=None}}
     }
    ]
   }
  }
 }
 ConversationLinkNode_0 {id=#4, class=ConversationLinkNode, properties {remoteEntryTag {tagName=Conversation.Entry.BasicServices}}}
 ConversationLinkNode_1 {id=#3, class=ConversationLinkNode, properties {remoteEntryTag {tagName=Conversation.Entry.NPCLibrary.Quests}}}
]
